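# Placeholder SoundCue for a game success sound among Battle Pass S24 quest actor effects.
# A single wave player reuses the TimedTrial_Success wave from the Athena quest sounds, with boosted volume in the Fort_Quests sound class.
# Attenuation overrides turn off distance falloff and spatialization, so it plays as non-positional audio.
# A concurrency override limits it to one instance at a time.
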SoundCue /BattlePassS24/Items/QuestActors/SFX/Placeholder/Proto_GameSuccess_Cue
root=WavePlayer_0
nodes [{id=WavePlayer_0, properties {soundWaveAssetPtr=/Game/Athena/Sounds/Quests/TimedTrial/TimedTrial_Success}}]
properties {
 bOverrideConcurrency=true
 maxDistance=2097152
 soundClassObject=&Fort_Quests
 volumeMultiplier=1.68
 attenuationOverrides {bAttenuate=false, bSpatialize=false}
 concurrencyOverrides {maxCount=1}
}
