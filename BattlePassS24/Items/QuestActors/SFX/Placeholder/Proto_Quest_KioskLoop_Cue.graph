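# Placeholder SoundCue for a looping quest kiosk sound in BattlePassS24, judging by its name and path.
# A single looping wave player, wired as the root, plays the TimedTrial_Music asset under the Fort_Quests sound class.
# Its own concurrency override uses PreventNew, so new instances are refused once the limit is reached.
# Attenuation override values are stored, but the attenuation override is not switched on.
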SoundCue /BattlePassS24/Items/QuestActors/SFX/Placeholder/Proto_Quest_KioskLoop_Cue
root=WavePlayer_1
nodes [{id=WavePlayer_1, properties {bLooping=true, soundWaveAssetPtr=/Game/Athena/Sounds/Quests/TimedTrial/TimedTrial_Music}}]
properties {
 bOverrideConcurrency=true
 maxDistance=2097152
 soundClassObject=&Fort_Quests
 volumeMultiplier=1
 attenuationOverrides {falloffDistance=1000, attenuationShapeExtents {x=200}}
 concurrencyOverrides {resolutionRule=PreventNew}
}
